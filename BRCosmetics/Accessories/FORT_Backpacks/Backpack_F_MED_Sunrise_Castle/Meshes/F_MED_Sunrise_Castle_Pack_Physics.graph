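# Physics asset for the Sunrise Castle backpack cosmetic, defining bodies and joints for its dangling cloth pieces.
# Two chains hang from small sphere base bodies: one box segment on top, two linked boxes below.
# Constraints give each segment moderate swing and very narrow twist, so the cloth bends with little roll.
# An unconstrained capsule on dyn_offset adds a larger volume, apparently for the cloth to collide against.
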PhysicsAsset /BRCosmetics/Accessories/FORT_Backpacks/Backpack_F_MED_Sunrise_Castle/Meshes/F_MED_Sunrise_Castle_Pack_Physics
bodies {
 dyn_cloth_01_lower [{
   shapeType=Box
   extentX=4.3155
   extentY=4.382
   extentZ=0.9652
   shapeName=dyn_cloth_01_lower_box
   center=<1.9835 0 0.2477>
   rotation=<rot 9.4778 0 0>
  }]
 dyn_cloth_base_lower [{shapeType=Sphere, radius=0.5, shapeName=dyn_cloth_base_lower_sphere}]
 dyn_cloth_base_top [{shapeType=Sphere, radius=0.5, shapeName=dyn_cloth_base_top_sphere}]
 dyn_cloth_parent_lower [{shapeType=Box, extentX=2.4664, extentY=3.7622, extentZ=1.0454, shapeName=dyn_cloth_parent_lower_box, center=<0.9836 0 0>}]
 dyn_cloth_parent_top [{
   shapeType=Box
   extentX=3.2788
   extentY=4.704
   extentZ=0.75
   shapeName=dyn_cloth_parent_top_box
   center=<1.6198 0.1538 0.3043>
   rotation=<rot 9.8567 4.0991 0.7029>
  }]
 dyn_offset [{shapeType=Capsule, length=8.5, radius=3.875, shapeName=dyn_offset_capsule, center=<0.9731 0.5802 -0.679>, rotation=<rot 0 0 88.4149>}]
}
constraints [
 {bone1Name=dyn_cloth_parent_top, bone2Name=dyn_cloth_base_top, swing1LimitDegrees=40, swing2LimitDegrees=40, twistLimitDegrees=3}
 {bone1Name=dyn_cloth_parent_lower, bone2Name=dyn_cloth_base_lower, swing1LimitDegrees=30, swing2LimitDegrees=30, twistLimitDegrees=3}
 {bone1Name=dyn_cloth_01_lower, bone2Name=dyn_cloth_parent_lower, swing1LimitDegrees=30, swing2LimitDegrees=40, twistLimitDegrees=3}
]
